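# Niagara system for the QuicheLorraineLime pickaxe trail, drawing two near-black anim-trail ribbons.
# SoftTrail is wider and shorter-lived, using the standard anim-trail modules, while Trail_Main initializes and updates through a shared system Socket Reader.
# Both multiply a particle alpha curve by an emitter-age fade, with lifetimes shortened in the front end.
# Trail_Main carries time-of-day RGB scaling inputs, but its RGB scaling is switched off.
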NiagaraSystem /BRCosmetics/Weapons/FORT_Melee/Pickaxe_QuicheLorraineLime/FX/NS_Pickaxe_QuicheLorraineLime_Trail
dependencies {
 usedDataInterfaces [&NiagaraDataInterfaceCurve, &NiagaraDataInterfaceSocketReader]
 usedDynamicInputs []
 usedModules [
  &self:SystemSpawnScript.SystemScriptSource.SystemScriptGraph.NiagaraNodeAssignment_2.SetVariables_#1
  &SystemState
 ]
}
emitters {
 SoftTrail {
  emitterUpdateScript [
   &EmitterState()
   &FN_SpawnAnimTrail(AnimTrailWidthMode=::FromFirst)
  ]
  particleSpawnScript [
   &FN_InitializeAnimTrailBoneIndecies()
   &FN_InitializeAnimTrail(
    AnimTrailWidth=3
    Color=<rgba 0.001 0.0006 0.0005 1>
    Lifetime=&FN_FrontEndFloatModifier(Gameplay=0.15, Front End=0.12)
   ) {module=FN_InitializeAnimTrail001}
  ]
  particleUpdateScript [
   &FN_UpdateAnimTrail() {module=FN_UpdateAnimTrail001}
   &ScaleColor(
    Scale Alpha=&Multiply_Float(
     A=&FloatFromCurve(
      FloatCurve=DataInterface() {propertyValues {Curve {keys [
          <at 0 0 RCIM_Cubic>
          {arriveTangent=-0.3998, interpMode=RCIM_Cubic, leaveTangent=-0.3998, tangentMode=RCTM_User, time=0.15, value=1}
          {arriveTangent=-2.0275, interpMode=RCIM_Cubic, leaveTangent=-2.0275, tangentMode=RCTM_User, time=1, value=0}
         ]}}}
      Scale Curve=0.7
     )
     B=&FloatFromCurve(
      FloatCurve=DataInterface() {propertyValues {Curve {
         defaultValue=1
         keys [{interpMode=RCIM_Cubic, tangentMode=RCTM_User, time=0.1, value=1}, {interpMode=RCIM_Cubic, tangentMode=RCTM_User, time=0.2, value=0}]
        }}}
      CurveIndex=$Emitter.Age
      Scale Curve=&RemapRange(
       Input Value To Remap=1
       Input Min=0
       Input Max=&FN_FrontEndFloatModifier(Gameplay=0.15, Front End=0.12)
       Output Min=0
       Output Max=1
      )
     )
    )
    ScaleA=true
    ScaleRGB=true
   ) {module=ScaleColor001}
  ]
  rendererClasses [&NiagaraRibbonRendererProperties]
  renderers [{rendererClass=&NiagaraRibbonRendererProperties}]
 }
 Trail_Main {
  emitterUpdateScript [
   &EmitterState()
   &FN_SpawnAnimTrail_v2()
  ]
  particleSpawnScript [
   /Game/Effects/Niagara/Modules/AnimTrail/FN_InitializeAnimTrail/FN_InitializeAnimTrailBoneIndecies(Second Bone or Socket Index=1)
   /Game/Effects/Niagara/Modules/AnimTrail/FN_InitializeAnimTrail/FN_InitializeAnimTrail_SocketReader/FN_InitializeAnimTrail_SocketReader(
    AnimTrailWidth=2
    Color:LinearColor=<rgba 0 0 0 1>
    Lifetime:NiagaraFloat=&FN_FrontEndFloatModifier(Gameplay=0.22, Front End=0.15)
    Socket Reader:NiagaraDataInterfaceSocketReader=$System.Socket Reader
   ) {module=FN_InitializeAnimTrail_SocketReader001}
  ]
  particleUpdateScript [
   /Game/Effects/Niagara/Modules/AnimTrail/FN_UpdateAnimTrail/FN_UpdateAnimTrail_SocketReader/FN_UpdateAnimTrail_SocketReader(Socket Reader:NiagaraDataInterfaceSocketReader=$System.Socket Reader)
   &ScaleColor(
    Scale Alpha=&Multiply_Float(
     A=&FloatFromCurve(
      FloatCurve=DataInterface() {propertyValues {Curve {keys [
          {interpMode=RCIM_Cubic, leaveTangent=0.0475, tangentMode=RCTM_Break, time=0, value=1}
          {arriveTangent=-0.151, tangentMode=RCTM_User, time=0.5, value=0}
         ]}}}
      Scale Curve=1.75
     )
     B=&FloatFromCurve(
      FloatCurve=DataInterface() {propertyValues {Curve {
         defaultValue=1
         keys [{interpMode=RCIM_Cubic, tangentMode=RCTM_User, time=0.3, value=1}, {interpMode=RCIM_Cubic, tangentMode=RCTM_User, time=0.7, value=0}]
        }}}
      CurveIndex=$Emitter.Age
      Scale Curve=&RemapRange(
       Input Value To Remap=1
       Input Min=0
       Input Max=&FN_FrontEndFloatModifier(Gameplay=0.22, Front End=0.15)
       Output Min=0
       Output Max=1
      )
     )
    )
    Scale RGB=&Multiply_Vector(
     A=&FN_Select_Vector_from_Time_Of_Day(Morning=<1 1 1>, Day=<1 1 1>, Evening=<1 1 1>, Night=<1 1 1>)
     B=&VectorFromFloat(Value=&FNTimeOfDayScalar(Night Scale=0.5, Morning Scale=0.6, Evening Scale=0.6))
    )
    ScaleA=true
    ScaleRGB=false
   )
  ]
  rendererClasses [&NiagaraRibbonRendererProperties]
  renderers [{rendererClass=&NiagaraRibbonRendererProperties}]
 }
}
summary {
 systemName=NS_Pickaxe_QuicheLorraineLime_Trail
 emitters [
  {emitterName=SoftTrail, rendererClasses [&NiagaraRibbonRendererProperties]}
  {emitterName=Trail_Main, rendererClasses [&NiagaraRibbonRendererProperties]}
 ]
 userVariables [
  User.bIsFrontEnd {type=NiagaraBool}
  User.bIsFrontEndPreview {type=NiagaraBool}
  User.RibbonIDOffset {type=NiagaraInt32}
  User.SkeletalMesh {
   type=NiagaraDataInterfaceSkeletalMesh
   defaultValue {
    struct=NiagaraExt_VariableValue_DataInterface
    value {dataInterface=&self:NiagaraDataInterfaceSkeletalMesh_0}
   }
  }
 ]
}
