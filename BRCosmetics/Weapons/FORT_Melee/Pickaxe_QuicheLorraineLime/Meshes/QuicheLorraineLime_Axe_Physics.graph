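# Physics asset for the QuicheLorraineLime pickaxe, defining capsule bodies and joints for a dynamic bag chain plus thigh and calf bodies.
# The bag runs from dym_bagBase through dym_bagRotate to three segments, with twist locked, most swing at the bagRotate joint, slight single-axis bend in the next two, and the last joint fully locked.
# Left and right calf-to-thigh joints swing and twist freely, their capsules likely serving as leg colliders for the bag.
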PhysicsAsset /BRCosmetics/Weapons/FORT_Melee/Pickaxe_QuicheLorraineLime/Meshes/QuicheLorraineLime_Axe_Physics
bodies {
 calf_l [{shapeType=Capsule, length=30, radius=8, shapeName=thigh_l_capsule, center=<-26 0 -4.2974>, rotation=<rot 90 180 0>}]
 calf_r [{shapeType=Capsule, length=30, radius=9.5, shapeName=thigh_r_capsule, center=<26 0 -1.3537>, rotation=<rot 90 0 0>}]
 dym_bagBase [{shapeType=Capsule, length=6.6284, radius=3.2616, shapeName=dym_bagBase_capsule, rotation=<rot 0 -90 0>}]
 dym_bagRotate [{
   shapeType=Capsule
   length=18.6306
   radius=6.016
   shapeName=dym_bagRotate_capsule
   center=<7.8308 0 -0.0863>
   rotation=<rot 0.0128 -90 0.6318>
  }]
 dym_bag_01 [{
   shapeType=Capsule
   length=22.0653
   radius=7.2617
   shapeName=dym_bagRotate_capsule
   center=<2.7639 0 -0.0305>
   rotation=<rot 0.0128 -90 0.6318>
  }]
 dym_bag_02 [{
   shapeType=Capsule
   length=24.4584
   radius=7.8593
   shapeName=dym_bagRotate_capsule
   center=<2.7639 0 -0.0305>
   rotation=<rot 0.0128 -90 0.6318>
  }]
 dym_bag_03 [{
   shapeType=Capsule
   length=23.5761
   radius=8.8975
   shapeName=dym_bagRotate_capsule
   center=<2.7639 0 -0.0323>
   rotation=<rot 0.0128 -90 0.6318>
  }]
 thigh_l [{shapeType=Capsule, length=34.6616, radius=9.5479, shapeName=thigh_l_capsule, center=<-22.1907 -0.6066 0.3009>, rotation=<rot 90 180 0>}]
 thigh_r [{shapeType=Capsule, length=34.6616, radius=9.5479, shapeName=thigh_r_capsule, center=<22.1907 0.6066 -0.3009>, rotation=<rot 90 0 0>}]
}
constraints [
 {
  bone1Name=calf_l
  bone2Name=thigh_l
  swing1LimitDegrees=45
  swing1Motion=Free
  swing2LimitDegrees=45
  swing2Motion=Free
  twistLimitDegrees=45
  twistMotion=Free
 }
 {
  bone1Name=calf_r
  bone2Name=thigh_r
  swing1LimitDegrees=45
  swing1Motion=Free
  swing2LimitDegrees=45
  swing2Motion=Free
  twistLimitDegrees=45
  twistMotion=Free
 }
 {bone1Name=dym_bagRotate, bone2Name=dym_bagBase, swing1LimitDegrees=30, swing2LimitDegrees=2.5, twistLimitDegrees=45, twistMotion=Locked}
 {
  bone1Name=dym_bag_01
  bone2Name=dym_bagRotate
  swing1LimitDegrees=2.5
  swing2LimitDegrees=0
  swing2Motion=Locked
  twistLimitDegrees=45
  twistMotion=Locked
 }
 {
  bone1Name=dym_bag_02
  bone2Name=dym_bag_01
  swing1LimitDegrees=2.5
  swing2LimitDegrees=0
  swing2Motion=Locked
  twistLimitDegrees=45
  twistMotion=Locked
 }
 {
  bone1Name=dym_bag_03
  bone2Name=dym_bag_02
  swing1LimitDegrees=1.5
  swing1Motion=Locked
  swing2LimitDegrees=0
  swing2Motion=Locked
  twistLimitDegrees=45
  twistMotion=Locked
 }
]
