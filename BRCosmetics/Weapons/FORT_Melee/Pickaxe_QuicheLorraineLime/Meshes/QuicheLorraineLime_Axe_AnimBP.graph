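# Anim Blueprint for the QuicheLorraineLime pickaxe that adds physics-driven secondary motion to its bag, zipper and buckles.
# The event graph caches the parent body mesh's root, pelvis, thigh and calf world transforms, and the AnimGraph has world-space replace nodes for those bones.
# A RigidBody pass, disabled when standalone in the front end, feeds a bag-base rotation offset, a no-body-collision RigidBody, and two chained rounds of AnimDynamics on zipper and buckles.
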AnimBlueprint /BRCosmetics/Weapons/FORT_Melee/Pickaxe_QuicheLorraineLime/Meshes/QuicheLorraineLime_Axe_AnimBP
target_skeleton=&/QuicheLorraineLime_Axe_Skeleton
graphs [
 AnimGraph {nodes [
   AnimGraphNode_SubInput_0 = Input Pose()
   n1 = "Transform (Modify) Bone - Bone: body_root"(ComponentPose=Local To Component(LocalPose=@AnimGraphNode_SubInput_0), Alpha=1, Translation="0,0,0", Rotation="0,0,0", Scale="1,1,1") {
    rotationMode=BMM_Replace
    rotationSpace=BCS_WorldSpace
    scaleSpace=BCS_BoneSpace
    translationMode=BMM_Replace
    translationSpace=BCS_WorldSpace
    alphaBoolBlend {blendOption=Linear}
    alphaScaleBiasClamp {bInterpResult=true, interpSpeedDecreasing=1, interpSpeedIncreasing=1, inRange {max=0, min=-40}, outRange {max=0, min=1}}
    boneToModify {boneName=body_root}
   }
   n2 = "Transform (Modify) Bone - Bone: pelvis"(ComponentPose=@n1, Alpha=1, Translation="0,0,0", Rotation="0,0,0", Scale="1,1,1") {
    rotationMode=BMM_Replace
    rotationSpace=BCS_WorldSpace
    scaleSpace=BCS_BoneSpace
    translationMode=BMM_Replace
    translationSpace=BCS_WorldSpace
    alphaBoolBlend {blendOption=Linear}
    alphaScaleBiasClamp {inRange {max=0, min=-40}, outRange {max=0, min=1}}
    boneToModify {boneName=pelvis}
   }
   n3 = "Transform (Modify) Bone - Bone: thigh_l"(ComponentPose=@n2, Alpha=1, Translation="0,0,0", Rotation="0,0,0", Scale="1,1,1") {
    rotationMode=BMM_Replace
    rotationSpace=BCS_WorldSpace
    scaleSpace=BCS_BoneSpace
    translationMode=BMM_Replace
    translationSpace=BCS_WorldSpace
    alphaBoolBlend {blendOption=Linear}
    alphaScaleBiasClamp {inRange {max=0, min=-40}, outRange {max=0, min=1}}
    boneToModify {boneName=thigh_l}
   }
   n4 = "Transform (Modify) Bone - Bone: thigh_r"(ComponentPose=@n3, Alpha=1, Translation="0,0,0", Rotation="0,0,0", Scale="1,1,1") {
    rotationMode=BMM_Replace
    rotationSpace=BCS_WorldSpace
    scaleSpace=BCS_BoneSpace
    translationMode=BMM_Replace
    translationSpace=BCS_WorldSpace
    alphaBoolBlend {blendOption=Linear}
    alphaScaleBiasClamp {inRange {max=0, min=-40}, outRange {max=0, min=1}}
    boneToModify {boneName=thigh_r}
   }
   n5 = "Transform (Modify) Bone - Bone: calf_l"(ComponentPose=@n4, Alpha=1, Translation="0,0,0", Rotation="0,0,0", Scale="1,1,1") {
    rotationMode=BMM_Replace
    rotationSpace=BCS_WorldSpace
    scaleSpace=BCS_BoneSpace
    translationMode=BMM_Replace
    translationSpace=BCS_WorldSpace
    alphaBoolBlend {blendOption=Linear}
    alphaScaleBiasClamp {inRange {max=0, min=-40}, outRange {max=0, min=1}}
    boneToModify {boneName=calf_l}
   }
   n6 = "Transform (Modify) Bone - Bone: calf_r"(ComponentPose=@n5, Alpha=1, Translation="0,0,0", Rotation="0,0,0", Scale="1,1,1") {
    rotationMode=BMM_Replace
    rotationSpace=BCS_WorldSpace
    scaleSpace=BCS_BoneSpace
    translationMode=BMM_Replace
    translationSpace=BCS_WorldSpace
    alphaBoolBlend {blendOption=Linear}
    alphaScaleBiasClamp {inRange {max=0, min=-40}, outRange {max=0, min=1}}
    boneToModify {boneName=calf_r}
   }
   n7 = RigidBody!(ComponentPose=@n6, bAlphaBoolEnabled:bool=NOT Boolean(A=Get StandaloneInFE() -> (StandaloneInFE:bool)), ExternalForce) -> (Pose) {
    alphaInputType=Bool
    bOverrideWorldGravity=true
    overridePhysicsAsset=&/QuicheLorraineLime_Axe_Physics
    componentAppliedLinearAccClamp=<3500 3500 3500>
    componentLinearAccScale=<0.3 0.6 0.3>
    componentLinearVelScale=<0.6 0.8 0.6>
    overrideWorldGravity {z=-980}
   }
   n8 = "Transform (Modify) Bone - Bone: dym_bagBase"!(ComponentPose=@n7, bAlphaBoolEnabled:bool=True, Translation="0,0,0", Rotation="0,-25,0", Scale="1,1,1") -> (Pose) {
    alphaInputType=Bool
    rotationMode=BMM_Additive
    rotationSpace=BCS_BoneSpace
    scaleSpace=BCS_BoneSpace
    translationSpace=BCS_BoneSpace
    alphaBoolBlend {blendOption=Linear}
    alphaScaleBiasClamp {inRange {max=0, min=-40}, outRange {max=0, min=1}}
    boneToModify {boneName=dym_bagBase}
    rotation {yaw=-25}
   }
   n9 = RigidBody(ComponentPose=@n8) {
    alphaInputType=Bool
    bOverrideWorldGravity=true
    overridePhysicsAsset=&/QuicheLorraineLime_Axe_NoBodyCollision_Physics
    componentAppliedLinearAccClamp=<3500 3500 3500>
    componentLinearAccScale=<2 2 2>
    componentLinearVelScale=<2 2 2>
    overrideWorldGravity {z=-1980}
   }
   n10 = Anim Dynamics - Bone: dym_zipper_01!(ComponentPose=@n9, bAlphaBoolEnabled:bool=True, ExternalForce="0,0,0") -> (Pose) {
    alphaInputType=Bool
    angularDampingOverride=0.999
    bOverrideAngularDamping=true
    bUseGravityOverride=true
    boundBone {boneName=dym_zipper_01}
    componentAppliedLinearAccClamp=<3500 3500 3500>
    componentLinearAccScale=<0.4 0.4 0.4>
    componentLinearVelScale=<0.8 0.8 0.8>
    gravityOverride {z=-500}
    physicsBodyDefinitions [{
      boundBone {boneName=dym_zipper_01}
      boxExtents=<10 10 10>
      constraintSetup {angularLimitsMax=<0 0 0>, angularLimitsMin=<0 -60 0>, angularTarget=<0 0 0>}
      localJointOffset=<3 0 0>
     }]
   }
   n11 = Anim Dynamics - Bone: dym_buck_01!(ComponentPose=@n10, bAlphaBoolEnabled:bool=True, ExternalForce="0,0,0") -> (Pose) {
    alphaInputType=Bool
    angularDampingOverride=0.99
    bOverrideAngularDamping=true
    bUseGravityOverride=true
    boundBone {boneName=dym_buck_01}
    componentAppliedLinearAccClamp=<3500 3500 3500>
    componentLinearAccScale=<0.3 0.3 0.3>
    componentLinearVelScale=<0.4 0.4 0.4>
    gravityOverride {z=-500}
    physicsBodyDefinitions [{
      boundBone {boneName=dym_buck_01}
      boxExtents=<20 10 10>
      constraintSetup {angularLimitsMax=<0 0.5 0>, angularLimitsMin=<0 -5 0>, angularTarget=<0 0 0>}
      localJointOffset=<3 0 0>
     }]
   }
   n12 = Anim Dynamics - Bone: dym_buck_02!(ComponentPose=@n11, bAlphaBoolEnabled:bool=True, ExternalForce="0,0,0") -> (Pose) {
    alphaInputType=Bool
    angularDampingOverride=0.99
    bOverrideAngularDamping=true
    bUseGravityOverride=true
    boundBone {boneName=dym_buck_02}
    componentAppliedLinearAccClamp=<3500 3500 3500>
    componentLinearAccScale=<0.3 0.3 0.3>
    componentLinearVelScale=<0.4 0.4 0.4>
    gravityOverride {z=-500}
    physicsBodyDefinitions [{
      boundBone {boneName=dym_buck_02}
      boxExtents=<20 10 10>
      constraintSetup {angularLimitsMax=<0 0.5 0>, angularLimitsMin=<0 -7.5 0>, angularTarget=<0 0 0>}
      localJointOffset=<3 0 0>
     }]
   }
   n13 = Anim Dynamics - Bone: dym_buck_03!(ComponentPose=@n12, bAlphaBoolEnabled:bool=True, ExternalForce="0,0,0") -> (Pose) {
    alphaInputType=Bool
    angularDampingOverride=0.99
    bOverrideAngularDamping=true
    bUseGravityOverride=true
    boundBone {boneName=dym_buck_03}
    componentAppliedLinearAccClamp=<3500 3500 3500>
    componentLinearAccScale=<0.3 0.3 0.3>
    componentLinearVelScale=<0.4 0.4 0.4>
    gravityOverride {z=-500}
    physicsBodyDefinitions [{
      boundBone {boneName=dym_buck_03}
      boxExtents=<20 10 10>
      constraintSetup {angularLimitsMax=<0 0.5 0>, angularLimitsMin=<0 -20 0>, angularTarget=<0 0 0>}
      localJointOffset=<3 0 0>
     }]
   }
   n14 = Anim Dynamics - Bone: dym_zipper_01(ComponentPose=@n13, bAlphaBoolEnabled=True, ExternalForce="0,0,0") {
    alphaInputType=Bool
    angularDampingOverride=0.999
    bOverrideAngularDamping=true
    bUseGravityOverride=true
    boundBone {boneName=dym_zipper_01}
    componentAppliedLinearAccClamp=<3500 3500 3500>
    componentLinearAccScale=<0.4 0.4 0.4>
    componentLinearVelScale=<0.8 0.8 0.8>
    gravityOverride {z=-980}
    physicsBodyDefinitions [{
      boundBone {boneName=dym_zipper_01}
      boxExtents=<20 10 10>
      constraintSetup {angularLimitsMax=<0 0 0>, angularLimitsMin=<0 -30 0>, angularTarget=<0 0 0>}
      localJointOffset=<3 0 0>
     }]
   }
   n15 = Anim Dynamics - Bone: dym_buck_01(ComponentPose=@n14, bAlphaBoolEnabled=True, ExternalForce="0,0,0") {
    alphaInputType=Bool
    angularDampingOverride=0.99
    bOverrideAngularDamping=true
    bUseGravityOverride=true
    boundBone {boneName=dym_buck_01}
    componentAppliedLinearAccClamp=<3500 3500 3500>
    componentLinearAccScale=<0.15 0.15 0.15>
    componentLinearVelScale=<0.2 0.2 0.2>
    gravityOverride {z=-980}
    physicsBodyDefinitions [{
      boundBone {boneName=dym_buck_01}
      boxExtents=<30 20 20>
      constraintSetup {angularLimitsMax=<0 0.5 0>, angularLimitsMin=<0 -2 0>, angularTarget=<0 0 0>}
      localJointOffset=<1.5 0 0>
     }]
   }
   n16 = Anim Dynamics - Bone: dym_buck_02(ComponentPose=@n15, bAlphaBoolEnabled=True, ExternalForce="0,0,0") {
    alphaInputType=Bool
    angularDampingOverride=0.99
    bOverrideAngularDamping=true
    bUseGravityOverride=true
    boundBone {boneName=dym_buck_02}
    componentAppliedLinearAccClamp=<3500 3500 3500>
    componentLinearAccScale=<0.15 0.15 0.15>
    componentLinearVelScale=<0.2 0.2 0.2>
    gravityOverride {z=-980}
    physicsBodyDefinitions [{
      boundBone {boneName=dym_buck_02}
      boxExtents=<30 20 20>
      constraintSetup {angularLimitsMax=<0 0.5 0>, angularLimitsMin=<0 -5 0>, angularTarget=<0 0 0>}
      localJointOffset=<1.5 0 0>
     }]
   }
   n17 = Anim Dynamics - Bone: dym_buck_03(ComponentPose=@n16, bAlphaBoolEnabled=True, ExternalForce="0,0,0") {
    alphaInputType=Bool
    angularDampingOverride=0.99
    bOverrideAngularDamping=true
    bUseGravityOverride=true
    boundBone {boneName=dym_buck_03}
    componentAppliedLinearAccClamp=<3500 3500 3500>
    componentLinearAccScale=<0.15 0.15 0.15>
    componentLinearVelScale=<0.2 0.2 0.2>
    gravityOverride {z=-980}
    physicsBodyDefinitions [{
      boundBone {boneName=dym_buck_03}
      boxExtents=<30 20 20>
      constraintSetup {angularLimitsMax=<0 0.5 0>, angularLimitsMin=<0 -2.5 0>, angularTarget=<0 0 0>}
      localJointOffset=<1.5 0 0>
     }]
   }
   Output Pose(Result=Component To Local(ComponentPose=@n17)) {layerGroup=None}
  ]}
 EventGraph {nodes [
   Event BlueprintUpdateAnimation()
   > n1 = Sequence!(execute) -> (then_0, then_1, then_2:Exec)
   n2 = GetPartSkeletalMeshComponent()
   > Is Valid(InputObject=@n2)
   n3 = GetAttachParent!(self:Scene Component Object Reference=@n2) -> (ReturnValue:Scene Component Object Reference)
   > Is Valid(InputObject=@n3)
   > Set Body_Mesh(Body_Mesh:Scene Component Object Reference=@n3) -> (Output_Get:Scene Component Object Reference)
   > n4 = Sequence!(execute) -> (then_0, then_1, then_2, then_3:Exec, then_4:Exec, then_5:Exec, then_6:Exec)
   n5 = Get Body_Mesh() -> (Body_Mesh:Scene Component Object Reference)
   n6 = BreakTransform(InTransform=GetSocketTransform(self=@n5, InSocketName=pelvis, TransformSpace=RTS_World))
   > Set pelvis_location(pelvis_location:Vector=@n6) -> (Output_Get:Vector)
   > Set pelvis_rotation(pelvis_rotation:Rotator=@n6.Rotation) -> (Output_Get:Rotator)
   n7 = BreakTransform(InTransform=GetSocketTransform(self=@n5, InSocketName=thigh_l, TransformSpace=RTS_World))
   Set thigh_left_location(execute=@n4.then_1, thigh_left_location:Vector=@n7) -> (Output_Get:Vector)
   > Set thigh_left_rotation(thigh_left_rotation:Rotator=@n7.Rotation) -> (Output_Get:Rotator)
   n8 = BreakTransform(InTransform=GetSocketTransform(self=@n5, InSocketName=thigh_r, TransformSpace=RTS_World))
   Set thigh_right_location(execute=@n4.then_2, thigh_right_location:Vector=@n8) -> (Output_Get:Vector)
   > Set thigh_right_rotation(thigh_right_rotation:Rotator=@n8.Rotation) -> (Output_Get:Rotator)
   n9 = BreakTransform(InTransform=GetSocketTransform(self=@n5, InSocketName=calf_r, TransformSpace=RTS_World))
   Set calf_right_location(execute=@n4.then_4, calf_right_location:Vector=@n9) -> (Output_Get:Vector)
   > Set calf_right_rotation(calf_right_rotation:Rotator=@n9.Rotation) -> (Output_Get:Rotator)
   n10 = BreakTransform(InTransform=GetSocketTransform(self=@n5, InSocketName=calf_l, TransformSpace=RTS_World))
   Set calf_left_location(execute=@n4.then_3, calf_left_location:Vector=@n10) -> (Output_Get:Vector)
   > Set calf_left_rotation(calf_left_rotation:Rotator=@n10.Rotation) -> (Output_Get:Rotator)
   n11 = BreakTransform(InTransform=GetSocketTransform(self=@n5, InSocketName=root, TransformSpace=RTS_World))
   Set body_root_location(execute=@n4.then_5, body_root_location:Vector=@n11) -> (Output_Get:Vector)
   > Set body_root_rotation(body_root_rotation:Rotator=@n11.Rotation) -> (Output_Get:Rotator)
   n12 = Get bIsInFrontEnd()
   Set IsNotInFE(execute=@n1.then_1, IsNotInFE:bool=NOT Boolean(A=@n12)) -> (Output_Get:bool)
   n13 = Get bIsOnPawn() -> (bIsOnPawn:bool)
   Set StandaloneInFE(execute=@n1.then_2, StandaloneInFE:bool=AND Boolean(A=@n12, B=NOT Boolean(A=@n13))) -> (Output_Get:bool)
   TryGetPawnOwner()
  ]}
]
variables [
 Body_Mesh {type=SceneComponent, default=null}
 body_root_location {type=Vector}
 body_root_rotation {type=Rotator}
 calf_left_location {type=Vector}
 calf_left_rotation {type=Rotator}
 calf_right_location {type=Vector}
 calf_right_rotation {type=Rotator}
 IsNotInFE {type=boolean}
 pelvis_location {type=Vector}
 pelvis_rotation {type=Rotator}
 StandaloneInFE {type=boolean}
 thigh_left_location {type=Vector}
 thigh_left_rotation {type=Rotator}
 thigh_right_location {type=Vector}
 thigh_right_rotation {type=Rotator}
]
